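; Conversation side effect that gives the talking player items rolled from a loot tier group, filed under the ExpertSale service effects.
; On the server side-effect call, it resolves the player participant from the conversation context and casts it to a Fort player controller.
; It then picks loot drops using the configured tier group name and loot level, casts each entry to an item definition, and adds one of each to the player's inventory.
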(fn AddItemToInstigatorInventory
  (params (InventoryOwner :type Fort%20Inventory%20Owner%20Interface%20Interface) (ItemDefinition :type Fort%20Item%20Definition%20Object%20Reference))
  (InventoryManagement|AddItem InventoryOwner ItemDefinition 1)
  (return))

(fn GetItemFromLootTierGroup
  (params (playerController :type Fort%20Player%20Controller%20Object%20Reference))
  (bind (_outloottodrop _returnvalue) (Loot|PickLootDrops (Variables|Default|GetLootTierGroupName) (Loot|GetLootLevel) -1 "(GameplayTags=)"))
  (if _returnvalue
    (for _array_element _outloottodrop
      (bind _asfort_item_definition (Utilities|Casting|CastToFortItemDefinition (Class|FortItemFunctionLibrary|BreakFortItemEntry _array_element)))
      (CallFunction|AddItemtoInstigatorInventory playerController _asfort_item_definition))
    (return true)
    (else
      (return false))))

(fn GetPlayerReference
  (params (Context :type Conversation%20Context%20Structure%20%28by%20ref%29))
  (bind _returnvalue (Class|FortConversationContextLibrary|GetParticipantActorPlayer Context))
  (Utilities|IsValid _returnvalue
    (:Is%20Valid
      (bind _asfort_player_controller (Utilities|Casting|CastToFortPlayerController _returnvalue)
        (:then
          (return _asfort_player_controller true))
        (:CastFailed
          _)))
    (:Is%20Not%20Valid
      _))
  (return 0 false))

(fn ServerCauseSideEffect
  (params (Context :type Conversation%20Context%20Structure%20%28by%20ref%29))
  (bind (_payload _success) (CallFunction|GetPlayerReference Context))
  (if _success
    (CallFunction|GetItemfromLootTierGroup _payload)
    (else
      (return))))
