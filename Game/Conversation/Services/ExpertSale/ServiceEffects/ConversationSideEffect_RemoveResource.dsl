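; Conversation side effect for the expert sale service that removes a configured amount of an item from the participating player's inventory on the server.
; ServerCauseSideEffect resolves the Fort player controller from the conversation context, then passes it to RemoveResourceFromPlayer.
; The removal is non-forced, with the amount read from a scalable float as an integer.
; A debug string prints the helper's result, which is always false.
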(fn GetPlayerController
  (params (Context :type Conversation%20Context%20Structure%20%28by%20ref%29))
  (bind _returnvalue (Class|FortConversationContextLibrary|GetParticipantActorPlayer Context))
  (Utilities|IsValid _returnvalue
    (:Is%20Valid
      (bind _asfort_player_controller (Utilities|Casting|CastToFortPlayerController _returnvalue)
        (:then
          (return _asfort_player_controller true))
        (:CastFailed
          _)))
    (:Is%20Not%20Valid
      _))
  (return 0 false))

(fn RemoveResourceFromPlayer
  (params (PlayerController :type Fort%20Player%20Controller%20Object%20Reference))
  (Inventory|RemoveItemFromPlayer :PlayerController PlayerController :ItemDefinition (Variables|Default|GetItem%20Definition) :AmountToRemove (ScalableFloat|GetValueasInteger (Variables|Default|GetamountToRemove)) :bForceRemoval false :bSkipVerbRemoval false)
  (return false))

(fn ServerCauseSideEffect
  (params (Context :type Conversation%20Context%20Structure%20%28by%20ref%29))
  (bind (_payload _success) (CallFunction|GetPlayerController Context))
  (if _success
    (bind _returnvalue (CallFunction|RemoveResourcefromPlayer _payload))
    (Development|PrintString (Utilities|String|ToString(Boolean) _returnvalue) true true "(R=1.000000,G=0.347529,B=0.000000,A=1.000000)" 1e+01)
    (return)
    (else
      (return))))
